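; Landing waypoint Blueprint for the Papaya skydive minigame, extending a parent waypoint class.
; Construction script gives the waypoint mesh a dynamic instance of the simple skydive ring material with a small X/Y-only WobbleScale.
; BeginPlay defers to the parent, while the ring-out animation event calls the parent and then hides the actor in game.
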(event EventBeginPlay
  (|CallParentFunction))

(event EventAnimateRingOut
  (|CallParentFunction)
  (Rendering|SetActorHiddenInGame true))

(fn ConstructionScript
  (bind _returnvalue (Rendering|Material|CreateDynamicMaterialInstance (Variables|Default|GetWayPointMesh) 0 "/Game/Athena/Items/QuestInteractables/SkydiveRings/FX/M_Athena_SkydiveRings_Simple.M_Athena_SkydiveRings_Simple" "None" :!class /Script/Engine.PrimitiveComponent))
  (Rendering|Material|SetVectorParameterValue _returnvalue "WobbleScale" "(R=0.050000,G=0.050000,B=0.000000,A=1.000000)" :!class /Script/Engine.MaterialInstanceDynamic))
